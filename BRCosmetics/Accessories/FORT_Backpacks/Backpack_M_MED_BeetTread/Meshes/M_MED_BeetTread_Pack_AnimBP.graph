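# Animation Blueprint for the BeetTread backpack, adding secondary motion to its flag and tassels.
# Additive roll corrections on the pack bone for pickaxe use and skydiving feed a cached pose.
# Two AnimDynamics chains simulate from that pose, selected by bIsInFrontEnd; other flag chains, including those using computed gravity override, are unconnected.
# The event graph derives spine-rotation alphas and drives material WPO motion and skydive-float cloth wind.
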AnimBlueprint /BRCosmetics/Accessories/FORT_Backpacks/Backpack_M_MED_BeetTread/Meshes/M_MED_BeetTread_Pack_AnimBP
target_skeleton=&/M_MED_BeetTread_Pack_Skeleton
graphs [
 AnimGraph {nodes [
   n1 = Anim Dynamics - Bone: dyn_flag_01(Alpha=1, ExternalForce="0,0,0") {
    angularDampingOverride=1
    angularSpringConstant=100
    bAngularSpring=true
    bEnableWind=true
    bOverrideAngularDamping=true
    bOverrideLinearDamping=true
    bUseGravityOverride=true
    gravityScale=0.56
    linearDampingOverride=1
    simulationSpace=RootRelative
    windScale=1000
    alphaBoolBlend {blendOption=Linear}
    boundBone {boneName=dyn_flag_01}
    componentAppliedLinearAccClamp=<300 300 300>
    componentLinearAccScale=<0.3 0.3 0.3>
    componentLinearVelScale=<0.4 0.4 0.4>
    gravityOverride {z=-500}
    physicsBodyDefinitions [{
      boundBone {boneName=dyn_flag_01}
      boxExtents=<2.5 3 3>
      constraintSetup {angularLimitsMax=<0 10 0>, angularLimitsMin=<0 -7 0>, angularTarget=<1 0 0>}
      localJointOffset=<2 2 0.4>
     }]
   }
   n2 = Anim Dynamics - Bone: dyn_flag_02(ComponentPose=@n1, Alpha=1, ExternalForce="0,0,0") {
    angularDampingOverride=1
    angularSpringConstant=100
    bAngularSpring=true
    bEnableWind=true
    bOverrideAngularDamping=true
    bOverrideLinearDamping=true
    bUseGravityOverride=true
    gravityScale=0.56
    linearDampingOverride=1
    simulationSpace=RootRelative
    windScale=1000
    alphaBoolBlend {blendOption=Linear}
    boundBone {boneName=dyn_flag_02}
    componentAppliedLinearAccClamp=<300 300 300>
    componentLinearAccScale=<0.4 0.4 0.4>
    componentLinearVelScale=<0.4 0.4 0.4>
    gravityOverride {z=-500}
    physicsBodyDefinitions [{
      boundBone {boneName=dyn_flag_02}
      boxExtents=<3 3 3>
      constraintSetup {angularLimitsMax=<0 14 0>, angularLimitsMin=<0 -10 0>, angularTarget=<1 0 0>}
      localJointOffset=<2 1.5 0.4>
     }]
   }
   n3 = Anim Dynamics - Bone: dyn_flag_03(ComponentPose=@n2, Alpha=1, ExternalForce="0,0,0") {
    angularDampingOverride=1
    angularSpringConstant=100
    bAngularSpring=true
    bEnableWind=true
    bOverrideAngularDamping=true
    bOverrideLinearDamping=true
    bUseGravityOverride=true
    gravityScale=0.56
    linearDampingOverride=1
    simulationSpace=RootRelative
    windScale=1000
    alphaBoolBlend {blendOption=Linear}
    boundBone {boneName=dyn_flag_03}
    componentAppliedLinearAccClamp=<300 300 300>
    componentLinearAccScale=<0.4 0.4 0.4>
    componentLinearVelScale=<0.4 0.4 0.4>
    gravityOverride {z=-500}
    physicsBodyDefinitions [{
      boundBone {boneName=dyn_flag_03}
      boxExtents=<5 5 5>
      constraintSetup {angularLimitsMax=<1.5 20 0>, angularLimitsMin=<-1.5 -15 0>, angularTarget=<1 0 0>}
      localJointOffset=<3 0.65 0.4>
     }]
   }
   Anim Dynamics - Bone: dyn_flag_04(ComponentPose=@n3, Alpha=1, ExternalForce="0,0,0") {
    angularDampingOverride=1
    angularSpringConstant=500
    bAngularSpring=true
    bEnableWind=true
    bOverrideAngularDamping=true
    bOverrideLinearDamping=true
    bUseGravityOverride=true
    gravityScale=0.56
    linearDampingOverride=1
    simulationSpace=RootRelative
    windScale=1000
    alphaBoolBlend {blendOption=Linear}
    boundBone {boneName=dyn_flag_04}
    componentAppliedLinearAccClamp=<300 300 300>
    componentLinearAccScale=<0.3 0.3 0.3>
    componentLinearVelScale=<0.4 0.4 0.4>
    physicsBodyDefinitions [{
      boundBone {boneName=dyn_flag_04}
      boxExtents=<7.5 7.5 7.5>
      constraintSetup {angularLimitsMax=<1.5 30 0>, angularLimitsMin=<-1.5 -20 0>, angularTarget=<1 0 0>}
      localJointOffset=<2 -0.2 0.4>
     }]
   }
   n4 = Get Gravity Override()
   n5 = Anim Dynamics - Bone: dyn_flag_01!(ComponentPose, Alpha=1, GravityOverride:Vector=@n4, ExternalForce="0,0,0") -> (Pose) {
    angularDampingOverride=1
    angularSpringConstant=500
    bAngularSpring=true
    bEnableWind=true
    bOverrideAngularDamping=true
    bOverrideLinearDamping=true
    bUseGravityOverride=true
    gravityScale=0.56
    linearDampingOverride=1
    simulationSpace=RootRelative
    windScale=1000
    alphaBoolBlend {blendOption=Linear}
    boundBone {boneName=dyn_flag_01}
    componentAppliedLinearAccClamp=<300 300 300>
    componentLinearAccScale=<0.3 0.3 0.3>
    componentLinearVelScale=<0.4 0.4 0.4>
    gravityOverride {z=-730}
    physicsBodyDefinitions [{
      boundBone {boneName=dyn_flag_01}
      boxExtents=<2.5 3 3>
      constraintSetup {angularLimitsMax=<0 10 0>, angularLimitsMin=<0 -7 0>, angularTarget=<1 0 0>}
      localJointOffset=<2 2 0.4>
     }]
   }
   n6 = Anim Dynamics - Bone: dyn_flag_02!(ComponentPose=@n5, Alpha=1, GravityOverride:Vector=@n4, ExternalForce="0,0,0") -> (Pose) {
    angularDampingOverride=1
    angularSpringConstant=500
    bAngularSpring=true
    bEnableWind=true
    bOverrideAngularDamping=true
    bOverrideLinearDamping=true
    bUseGravityOverride=true
    gravityScale=0.56
    linearDampingOverride=1
    simulationSpace=RootRelative
    windScale=1000
    alphaBoolBlend {blendOption=Linear}
    boundBone {boneName=dyn_flag_02}
    componentAppliedLinearAccClamp=<300 300 300>
    componentLinearAccScale=<0.3 0.3 0.3>
    componentLinearVelScale=<0.4 0.4 0.4>
    gravityOverride {z=-700}
    physicsBodyDefinitions [{
      boundBone {boneName=dyn_flag_02}
      boxExtents=<3.5 3.5 3.5>
      constraintSetup {angularLimitsMax=<0 14 0>, angularLimitsMin=<0 -10 0>, angularTarget=<1 0 0>}
      localJointOffset=<2 1.5 0.4>
     }]
   }
   n7 = Anim Dynamics - Bone: dyn_flag_03!(ComponentPose=@n6, Alpha=1, GravityOverride:Vector=@n4, ExternalForce="0,0,0") -> (Pose) {
    angularDampingOverride=1
    angularSpringConstant=500
    bAngularSpring=true
    bEnableWind=true
    bOverrideAngularDamping=true
    bOverrideLinearDamping=true
    bUseGravityOverride=true
    gravityScale=0.56
    linearDampingOverride=1
    simulationSpace=RootRelative
    windScale=1000
    alphaBoolBlend {blendOption=Linear}
    boundBone {boneName=dyn_flag_03}
    componentAppliedLinearAccClamp=<300 300 300>
    componentLinearAccScale=<0.3 0.3 0.3>
    componentLinearVelScale=<0.4 0.4 0.4>
    gravityOverride {z=-680}
    physicsBodyDefinitions [{
      boundBone {boneName=dyn_flag_03}
      boxExtents=<4.5 4.5 4.5>
      constraintSetup {angularLimitsMax=<1.5 20 0>, angularLimitsMin=<-1.5 -15 0>, angularTarget=<1 0 0>}
      localJointOffset=<2 0.65 0.4>
     }]
   }
   Anim Dynamics - Bone: dyn_flag_04!(ComponentPose=@n7, Alpha=1, GravityOverride:Vector=@n4, ExternalForce="0,0,0") -> (Pose) {
    angularDampingOverride=1
    angularSpringConstant=500
    bAngularSpring=true
    bEnableWind=true
    bOverrideAngularDamping=true
    bOverrideLinearDamping=true
    bUseGravityOverride=true
    gravityScale=0.56
    linearDampingOverride=1
    simulationSpace=RootRelative
    windScale=1000
    alphaBoolBlend {blendOption=Linear}
    boundBone {boneName=dyn_flag_04}
    componentAppliedLinearAccClamp=<300 300 300>
    componentLinearAccScale=<0.3 0.3 0.3>
    componentLinearVelScale=<0.4 0.4 0.4>
    gravityOverride {z=-680}
    physicsBodyDefinitions [{
      boundBone {boneName=dyn_flag_04}
      boxExtents=<7.5 7.5 7.5>
      constraintSetup {angularLimitsMax=<1.5 30 0>, angularLimitsMin=<-1.5 -20 0>, angularTarget=<1 0 0>}
      localJointOffset=<2 -0.2 0.4>
     }]
   }
   n8 = Anim Dynamics - Bone: dyn_flag_01!(ComponentPose=Local To Component(LocalPose=Use cached pose 'input pose'()), Alpha=1, ExternalForce="0,0,0") -> (Pose) {
    angularDampingOverride=1
    angularSpringConstant=100
    bAngularSpring=true
    bEnableWind=true
    bOverrideAngularDamping=true
    bOverrideLinearDamping=true
    bUseGravityOverride=true
    gravityScale=0.56
    linearDampingOverride=1
    simulationSpace=RootRelative
    windScale=1000
    alphaBoolBlend {blendOption=Linear}
    boundBone {boneName=dyn_flag_01}
    componentAppliedLinearAccClamp=<300 300 300>
    componentLinearAccScale=<0.3 0.3 0.3>
    componentLinearVelScale=<0.4 0.4 0.4>
    gravityOverride {z=-500}
    physicsBodyDefinitions [{
      boundBone {boneName=dyn_flag_01}
      boxExtents=<2.5 3 3>
      constraintSetup {angularLimitsMax=<0 10 0>, angularLimitsMin=<0 -7 0>, angularTarget=<1 0 0>}
      localJointOffset=<2 0 0>
     }]
   }
   n9 = Anim Dynamics - Bone: dyn_flag_02!(ComponentPose=@n8, Alpha=1, ExternalForce="0,0,0") -> (Pose) {
    angularDampingOverride=1
    angularSpringConstant=100
    bAngularSpring=true
    bEnableWind=true
    bOverrideAngularDamping=true
    bOverrideLinearDamping=true
    bUseGravityOverride=true
    gravityScale=0.56
    linearDampingOverride=1
    simulationSpace=RootRelative
    windScale=1000
    alphaBoolBlend {blendOption=Linear}
    boundBone {boneName=dyn_flag_02}
    componentAppliedLinearAccClamp=<300 300 300>
    componentLinearAccScale=<0.4 0.4 0.4>
    componentLinearVelScale=<0.4 0.4 0.4>
    gravityOverride {z=-500}
    physicsBodyDefinitions [{
      boundBone {boneName=dyn_flag_02}
      boxExtents=<3 3 3>
      constraintSetup {angularLimitsMax=<0 14 0>, angularLimitsMin=<0 -10 0>, angularTarget=<1 0 0>}
      localJointOffset=<2 0 0>
     }]
   }
   n10 = Anim Dynamics - Bone: dyn_flag_03!(ComponentPose=@n9, Alpha=1, ExternalForce="0,0,0") -> (Pose) {
    angularDampingOverride=1
    angularSpringConstant=100
    bAngularSpring=true
    bEnableWind=true
    bOverrideAngularDamping=true
    bOverrideLinearDamping=true
    bUseGravityOverride=true
    gravityScale=0.56
    linearDampingOverride=1
    simulationSpace=RootRelative
    windScale=1000
    alphaBoolBlend {blendOption=Linear}
    boundBone {boneName=dyn_flag_03}
    componentAppliedLinearAccClamp=<300 300 300>
    componentLinearAccScale=<0.4 0.4 0.4>
    componentLinearVelScale=<0.4 0.4 0.4>
    gravityOverride {z=-500}
    physicsBodyDefinitions [{
      boundBone {boneName=dyn_flag_03}
      boxExtents=<5 5 5>
      constraintSetup {angularLimitsMax=<1.5 20 0>, angularLimitsMin=<-1.5 -15 0>, angularTarget=<1 0 0>}
      localJointOffset=<3 0 0>
     }]
   }
   n11 = Anim Dynamics - Bone: dyn_flag_04!(ComponentPose=@n10, Alpha=1, ExternalForce="0,0,0") -> (Pose) {
    angularDampingOverride=1
    angularSpringConstant=500
    bAngularSpring=true
    bEnableWind=true
    bOverrideAngularDamping=true
    bOverrideLinearDamping=true
    bUseGravityOverride=true
    gravityScale=0.56
    linearDampingOverride=1
    simulationSpace=RootRelative
    windScale=1000
    alphaBoolBlend {blendOption=Linear}
    boundBone {boneName=dyn_flag_04}
    componentAppliedLinearAccClamp=<300 300 300>
    componentLinearAccScale=<0.3 0.3 0.3>
    componentLinearVelScale=<0.4 0.4 0.4>
    physicsBodyDefinitions [{
      boundBone {boneName=dyn_flag_04}
      boxExtents=<7.5 7.5 7.5>
      constraintSetup {angularLimitsMax=<1.5 30 0>, angularLimitsMin=<-1.5 -20 0>, angularTarget=<1 0 0>}
      localJointOffset=<2 0 0>
     }]
   }
   n12 = Anim Dynamics - Bone: dyn_tassel_01(ComponentPose=@n11, Alpha=1, ExternalForce="0,0,0") {
    angularDampingOverride=0.99
    bOverrideAngularDamping=true
    bUseGravityOverride=true
    bUsePlanarLimit=false
    alphaBoolBlend {blendOption=Linear}
    boundBone {boneName=dyn_tassel_01}
    componentAppliedLinearAccClamp=<1000 1000 500>
    componentLinearAccScale=<0.5 0.5 0.2>
    componentLinearVelScale=<0.5 0.5 0.2>
    gravityOverride {z=-980}
    physicsBodyDefinitions [{
      boundBone {boneName=dyn_tassel_01}
      boxExtents=<10 8 10>
      constraintSetup {angularLimitsMax=<5 10 40>, angularLimitsMin=<-5 -2 -40>, angularTarget=<0 0 0>}
      localJointOffset=<0.5 0 0>
     }]
   }
   n13 = Anim Dynamics - Bone: dyn_tassel_02(ComponentPose=@n12, Alpha=1, ExternalForce="0,0,0") {
    angularDampingOverride=0.99
    bOverrideAngularDamping=true
    bUseGravityOverride=true
    bUsePlanarLimit=false
    alphaBoolBlend {blendOption=Linear}
    boundBone {boneName=dyn_tassel_02}
    componentAppliedLinearAccClamp=<800 800 500>
    componentLinearAccScale=<0.5 0.5 0.2>
    componentLinearVelScale=<0.5 0.5 0.2>
    gravityOverride {z=-980}
    physicsBodyDefinitions [{
      boundBone {boneName=dyn_tassel_02}
      boxExtents=<12 8 12>
      constraintSetup {angularLimitsMax=<5 10 40>, angularLimitsMin=<-5 -2 -40>, angularTarget=<0 0 0>}
      localJointOffset=<1 0 0>
     }]
   }
   n14 = Anim Dynamics - Bone: dyn_flag_01(ComponentPose=Local To Component(LocalPose=Use cached pose 'input pose'()), Alpha=1, ExternalForce="0,0,0") {
    angularDampingOverride=1
    angularSpringConstant=500
    bAngularSpring=true
    bEnableWind=true
    bOverrideAngularDamping=true
    bOverrideLinearDamping=true
    bUseGravityOverride=true
    gravityScale=0.56
    linearDampingOverride=1
    simulationSpace=RootRelative
    windScale=1000
    alphaBoolBlend {blendOption=Linear}
    boundBone {boneName=dyn_flag_01}
    componentAppliedLinearAccClamp=<300 300 300>
    componentLinearAccScale=<0.3 0.3 0.3>
    componentLinearVelScale=<0.4 0.4 0.4>
    gravityOverride {z=-50}
    physicsBodyDefinitions [{
      boundBone {boneName=dyn_flag_01}
      boxExtents=<2.5 3 3>
      constraintSetup {angularLimitsMax=<0 10 0>, angularLimitsMin=<0 -7 0>, angularTarget=<1 0 0>}
      localJointOffset=<2 2 0.4>
     }]
   }
   n15 = Anim Dynamics - Bone: dyn_flag_02(ComponentPose=@n14, Alpha=1, ExternalForce="0,0,0") {
    angularDampingOverride=1
    angularSpringConstant=500
    bAngularSpring=true
    bEnableWind=true
    bOverrideAngularDamping=true
    bOverrideLinearDamping=true
    bUseGravityOverride=true
    gravityScale=0.56
    linearDampingOverride=1
    simulationSpace=RootRelative
    windScale=1000
    alphaBoolBlend {blendOption=Linear}
    boundBone {boneName=dyn_flag_02}
    componentAppliedLinearAccClamp=<300 300 300>
    componentLinearAccScale=<0.3 0.3 0.3>
    componentLinearVelScale=<0.4 0.4 0.4>
    gravityOverride {z=-50}
    physicsBodyDefinitions [{
      boundBone {boneName=dyn_flag_02}
      boxExtents=<3.5 3.5 3.5>
      constraintSetup {angularLimitsMax=<0 14 0>, angularLimitsMin=<0 -10 0>, angularTarget=<1 0 0>}
      localJointOffset=<2 1.5 0.4>
     }]
   }
   n16 = Anim Dynamics - Bone: dyn_flag_03(ComponentPose=@n15, Alpha=1, ExternalForce="0,0,0") {
    angularDampingOverride=1
    angularSpringConstant=500
    bAngularSpring=true
    bEnableWind=true
    bOverrideAngularDamping=true
    bOverrideLinearDamping=true
    bUseGravityOverride=true
    gravityScale=0.56
    linearDampingOverride=1
    simulationSpace=RootRelative
    windScale=1000
    alphaBoolBlend {blendOption=Linear}
    boundBone {boneName=dyn_flag_03}
    componentAppliedLinearAccClamp=<300 300 300>
    componentLinearAccScale=<0.3 0.3 0.3>
    componentLinearVelScale=<0.4 0.4 0.4>
    gravityOverride {z=-50}
    physicsBodyDefinitions [{
      boundBone {boneName=dyn_flag_03}
      boxExtents=<4.5 4.5 4.5>
      constraintSetup {angularLimitsMax=<1.5 20 0>, angularLimitsMin=<-1.5 -15 0>, angularTarget=<1 0 0>}
      localJointOffset=<2 0.65 0.4>
     }]
   }
   n17 = Anim Dynamics - Bone: dyn_flag_04(ComponentPose=@n16, Alpha=1, ExternalForce="0,0,0") {
    angularDampingOverride=1
    angularSpringConstant=500
    bAngularSpring=true
    bEnableWind=true
    bOverrideAngularDamping=true
    bOverrideLinearDamping=true
    bUseGravityOverride=true
    gravityScale=0.56
    linearDampingOverride=1
    simulationSpace=RootRelative
    windScale=1000
    alphaBoolBlend {blendOption=Linear}
    boundBone {boneName=dyn_flag_04}
    componentAppliedLinearAccClamp=<300 300 300>
    componentLinearAccScale=<0.3 0.3 0.3>
    componentLinearVelScale=<0.4 0.4 0.4>
    gravityOverride {z=-50}
    physicsBodyDefinitions [{
      boundBone {boneName=dyn_flag_04}
      boxExtents=<7.5 7.5 7.5>
      constraintSetup {angularLimitsMax=<1.5 30 0>, angularLimitsMin=<-1.5 -20 0>, angularTarget=<1 0 0>}
      localJointOffset=<2 -0.2 0.4>
     }]
   }
   n18 = Anim Dynamics - Bone: dyn_tassel_01(ComponentPose=@n17, Alpha=1, ExternalForce="0,0,0") {
    angularDampingOverride=0.992
    bOverrideAngularDamping=true
    bUseGravityOverride=true
    bUsePlanarLimit=false
    alphaBoolBlend {blendOption=Linear}
    boundBone {boneName=dyn_tassel_01}
    componentAppliedLinearAccClamp=<1000 1000 1000>
    componentLinearAccScale=<0.5 0.5 0.5>
    componentLinearVelScale=<0.5 0.5 0.5>
    gravityOverride {z=-980}
    physicsBodyDefinitions [{
      boundBone {boneName=dyn_tassel_01}
      boxExtents=<8 12 12>
      constraintSetup {angularLimitsMax=<5 30 30>, angularLimitsMin=<-5 -2 -30>, angularTarget=<0 0 0>}
      localJointOffset=<1 0 0>
     }]
   }
   n19 = Anim Dynamics - Bone: dyn_tassel_02(ComponentPose=@n18, Alpha=1, ExternalForce="0,0,0") {
    angularDampingOverride=0.992
    bOverrideAngularDamping=true
    bUseGravityOverride=true
    bUsePlanarLimit=false
    alphaBoolBlend {blendOption=Linear}
    boundBone {boneName=dyn_tassel_02}
    componentAppliedLinearAccClamp=<800 800 800>
    componentLinearAccScale=<0.5 0.5 0.5>
    componentLinearVelScale=<0.5 0.5 0.5>
    gravityOverride {z=-980}
    physicsBodyDefinitions [{
      boundBone {boneName=dyn_tassel_02}
      boxExtents=<8 10 10>
      constraintSetup {angularLimitsMax=<5 25 25>, angularLimitsMin=<-5 -2 -25>, angularTarget=<0 0 0>}
      localJointOffset=<1 0 0>
     }]
   }
   Output Pose(Result=Blend Poses by bool(BlendPose_0=Component To Local(ComponentPose=@n13), BlendPose_1=Component To Local(ComponentPose=@n19), bActiveValue=Get bIsInFrontEnd()) {blendType=Linear}) {layerGroup=None}
   n20 = "Transform (Modify) Bone - Bone: pack"(
    ComponentPose=Local To Component(LocalPose=Input Pose())
    Alpha=Get own pickaxe fix() -> (own pickaxe fix:float)
    Translation="0,0,0"
    Rotation="0,0,5"
   ) {
    rotationMode=BMM_Additive
    rotationSpace=BCS_BoneSpace
    translationMode=BMM_Additive
    translationSpace=BCS_BoneSpace
    alphaBoolBlend {blendOption=Linear}
    boneToModify {boneName=pack}
    rotation {roll=5}
   }
   n21 = "Transform (Modify) Bone - Bone: pack"(ComponentPose=@n20, Alpha=Get regular pickaxe fix() -> (regular pickaxe fix:double), Translation="0,0,0", Rotation="0,0,15") {
    rotationMode=BMM_Additive
    rotationSpace=BCS_BoneSpace
    translationMode=BMM_Additive
    translationSpace=BCS_BoneSpace
    alphaBoolBlend {blendInTime=1, blendOption=Linear, blendOutTime=1}
    alphaScaleBiasClamp {bInterpResult=true}
    boneToModify {boneName=pack}
    rotation {roll=15}
   }
   n22 = "Transform (Modify) Bone - Bone: pack"!(ComponentPose=@n21, bAlphaBoolEnabled:bool=Get bIsSkyDiving(), Translation="0,0,0", Rotation="0,0,5", Scale) -> (Pose) {
    alphaInputType=Bool
    rotationMode=BMM_Additive
    rotationSpace=BCS_BoneSpace
    translationMode=BMM_Additive
    translationSpace=BCS_BoneSpace
    alphaBoolBlend {blendInTime=0.3, blendOption=Linear, blendOutTime=0.3}
    boneToModify {boneName=pack}
    rotation {roll=5}
   }
   Save cached pose 'input pose'(Pose=Component To Local(ComponentPose=@n22)) {cacheName=input pose}
  ]}
 EventGraph {nodes [
   n1 = Event BlueprintUpdateAnimation()
   > Set Delta Time X(Delta Time X=@n1.DeltaTimeX)
   > n2 = Sequence!(execute) -> (then_0, then_1, then_2:Exec, then_3:Exec)
   > Is Valid(InputObject=Get BodyMesh() -> (BodyMesh:Skeletal Mesh Component Object Reference))
   Is Valid(exec=@n2.then_1, InputObject=Get PlayerPawn() -> (PlayerPawn:Pawn Object Reference))
   > Set WPO Globals Function(
    Material=Get Mat Instance Dynamic()
    Overlay Material=Get Overlay Mat Dynamic() -> (Overlay Mat Dynamic:Material Instance Dynamic Object Reference)
    Skeletal Mesh=Get SkelMesh()
    "Local Noise Transform  Location"="Get Local Noise Transform  Location"() -> ("Local Noise Transform  Location":Vector)
    Local Noise Offset=Get Local Noise Offset()
   )
   > Set WPO Regional Function!(
    execute
    Material:Material Instance Dynamic Object Reference=Get Mat Instance Dynamic()
    Overlay Material:Material Instance Dynamic Object Reference=Get Overlay Mat Dynamic()
    Skeletal Mesh:Skeletal Mesh Component Object Reference=Get SkelMesh()
    Velocity Transform:Name=pack
    Velocity Interp Speed:double=Get Velocity Interp Speed()
    Collision Transform:Name=pack
    Collision Transform Axis:EAxis Enum
    Invert Collision Transform Axis:bool=false
    Delta Time X:double=Get Delta Time X()
    Previous Velocity Transform Location:"Vector (by ref)"=Get R1_Previous Velocity Transform Location() -> (R1_Previous Velocity Transform Location:Vector)
    Previous Velocity Transform Velocity:"Vector (by ref)"=Get R1_Previous Velocity Transform Velocity() -> (R1_Previous Velocity Transform Velocity:Vector)
    Region:int=1
   ) -> (then)
   > SetScalarParameterValue(self=Get Mat Instance Dynamic(), ParameterName=ClothSim_WindStrength, Value=SelectFloat(A=1, B=0, bPickA=Get bIsSkydiveFloating()))
   n3 = GetOwnerSkeletalMeshComponent()
   n4 = GetAlphaFromReferenceTransformDelta(SkeletalMeshComponent=@n3, BoneName=spine_05, BaseName=spine_01, InputRange=(X=10,Y=-10), TwistAxis="1,0,0")
   n5 = Get bIsSkydiveFloating()
   Set spine rotation alpha(execute=@n2.then_2, spine rotation alpha:float="Clamp (Float)"(Value=float - float(A=@n4, B:bool=@n5), Min=0, Max=1)) -> (Output_Get:float)
   n6 = GetAlphaFromReferenceTransformDelta(SkeletalMeshComponent=@n3, BoneName=spine_05, BaseName=spine_01, DeltaComponentType=EulerX, InputRange=(X=-80,Y=-130), TwistAxis="1,0,0")
   n7 = Get bIsSkydiveFloating()
   > Set spine rotation x(spine rotation x:float="Clamp (Float)"(Value=float - float(A=@n6, B=@n7), Min=0, Max=1)) -> (Output_Get:float)
   n8 = "To Float (Boolean)"(InBool="In Range (Float)"!(
     Value:double=GetAlphaFromReferenceTransformDelta(SkeletalMeshComponent=@n3, BoneName=spine_05, BaseName=spine_01, InputRange=(X=-10,Y=10), TwistAxis="1,0,0").RawOutput
     Min:double=-10
     Max:double=30
     InclusiveMin:bool=true
     InclusiveMax:bool=true
    ) -> (ReturnValue:bool))
   n9 = Get bIsSkydiveFloating()
   > Set spine rotation own(spine rotation own:float="Clamp (Float)"(Value=float - float(A=@n8, B=@n9), Min=0, Max=1)) -> (Output_Get:float)
   n10 = GetAlphaFromReferenceTransformDelta(SkeletalMeshComponent=@n3, BoneName=spine_05, BaseName=spine_01, InputRange=(X=-10,Y=-18), TwistAxis="1,0,0")
   n11 = Get bIsSkydiveFloating()
   > Set spine rotation own pickaxe(spine rotation own pickaxe:float="Clamp (Float)"(Value=float - float(A=@n10, B=@n11), Min=0, Max=1)) -> (Output_Get:float)
   n12 = float * float(
    A=GetAlphaFromReferenceTransformDelta(SkeletalMeshComponent=@n3, BoneName=spine_05, BaseName=spine_01, InputRange=(X=0,Y=-15), TwistAxis="1,0,0")
    B:bool=Montage_IsPlaying(
     self=GetOwnerFortPlayerAnimInstance!(self:Custom Character Part Anim Instance Object Reference) -> (ReturnValue:Fort Player Anim Instance Object Reference)
     Montage=/Game/Animation/Game/MainPlayer/Montages/Melee/PlayerMeleeCombo
    )
   ) {is=title K2Node_PromotableOperator}
   n13 = Get bIsSkydiveFloating()
   > Set regular pickaxe fix(regular pickaxe fix:double="Clamp (Float)"(Value=float - float(A=@n12, B=@n13), Min=0, Max=1)) -> (Output_Get:double)
   n14 = Gravity Override State Function DEPRECATED(
    execute=@n2.then_3
    CharacterPartAnimInstance=Self-Reference()
    GameTimeInSeconds=Get Delta Time X()
    GravityOverrideParams=Get Gravity Override Params()
   )
   > Set Gravity Override(Gravity Override=@n14)
   Event BlueprintInitializeAnimation()
   > Set SkelMesh(SkelMesh=GetOwnerSkeletalMeshComponent())
   n15 = TryGetPawnOwner()
   > Is Valid(InputObject=@n15)
   > Set PlayerPawn(PlayerPawn:Pawn Object Reference=@n15) -> (Output_Get:Pawn Object Reference)
   n16 = Event OnCharacterCustomizationLoaded()
   > Set backpack mesh(backpack mesh:Skeletal Mesh Component Budgeted Object Reference=Get Array Item with EFort Custom Part Type(Array=@n16.CustomMeshParts, PartType=Backpack)) -> (Output_Get:Skeletal Mesh Component Budgeted Object Reference)
   > Set BodyMesh(BodyMesh=Get Array Item with EFort Custom Part Type(Array=@n16.CustomMeshParts))
   n17 = GetPartSkeletalMeshComponent()
   > n18 = Get Dynamic Material Instance(SkeletalMeshComponent=@n17)
   > Set Mat Instance Dynamic(Mat Instance Dynamic=@n18)
   > n19 = GetOrCreateDynamicOverlayMaterial(MeshComponent=@n17, MaterialName=None)
   > Set Overlay Mat Dynamic(Overlay Mat Dynamic:Material Instance Dynamic Object Reference=@n19) -> (Output_Get:Material Instance Dynamic Object Reference)
   TryGetPawnOwner()
  ]}
]
variables [
 backpack mesh {type=SkeletalMeshComponentBudgeted}
 BodyMesh {type=SkeletalMeshComponent}
 Delta Time X {type=number}
 Gravity Override {type=Vector}
 Gravity Override Params {type=GravityOverrideParamsStruct}
 Local Noise Offset {type=Vector}
 "Local Noise Transform  Location" {type=Vector}
 Mat Instance Dynamic {type=MaterialInstanceDynamic}
 Overlay Mat Dynamic {type=MaterialInstanceDynamic}
 own pickaxe fix {type=number}
 PlayerPawn {type=Pawn, default=null}
 R1_Previous Velocity Transform Location {type=Vector}
 R1_Previous Velocity Transform Velocity {type=Vector}
 regular pickaxe fix {type=number}
 SkelMesh {type=SkeletalMeshComponent}
 spine rotation alpha {type=number}
 spine rotation own {type=number}
 spine rotation own pickaxe {type=number}
 spine rotation x {type=number}
 Velocity Interp Speed {type=number}
]
